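# Physics asset for the Maser vehicle body cosmetic, giving its skeletal mesh a simplified collision shape.
# It holds a single body on the root bone, made of one box sized and raised to enclose the car body.
# No constraints are defined, so the body acts as one rigid piece with no articulated or ragdoll parts.
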PhysicsAsset /VehicleCosmetics/Bodies/Maser/Meshes/Body_Maser_PhysicsDM
bodies {root [{shapeType=Box, extentX=444, extentY=216, extentZ=116, shapeName=root_box, center=<0.5136 0 95.4093>}]}
constraints []
